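; Rift shard jar consumable weapon Blueprint that swaps its mesh's flare FX material for a blank one around vehicle use.
; BeginPlay async-loads the blank material, caches the instigator pawn and binds its vehicle enter, exit and holster events.
; Entering blanks material slot 0 and exiting restores the flare.
; On holster events, gameplay tags and seat index restore the flare in helicopters and rear pickup truck seats, blanking it in front seats.
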(event EventBeginPlay
  ;[1 1 1] Init materials and events on begin play
  (|CallParentFunction)
  (bind _object (Utilities|AsyncLoadAsset (Variables|Default|GetBlankMaterial))
    (:Completed
      (bind _asmaterial_interface (Utilities|Casting|CastToMaterialInterface _object))
      (Variables|Default|SetNullMaterial _asmaterial_interface)
      (bind _asplayer_pawn_athena_generic_parent (Utilities|Casting|CastToPlayerPawn_Athena_Generic_Parent (Game|GetInstigator)))
      (Variables|Default|SetPlayerPawn _asplayer_pawn_athena_generic_parent)
      (Default|AssignOnEnteredVehicleDispatcher (Variables|Default|GetPlayerPawn) (delegate Custom|EnterVehicleEvent))
      (Default|BindEventtoOnExitedVehicleDispatcher (Variables|Default|GetPlayerPawn) (delegate Custom|ExitVehicleEvent))
      (EventDispatchers|BindEventtoOnHolsteredEvent (Variables|Default|GetPlayerPawn) (EventDispatchers|CreateEvent :!event Custom|OnHolsterSwapMaterials)))))

(event Custom|EnterVehicleEvent
  ;[1 1 1] Init materials and events on begin play
  (CallFunction|SetMaterial false))

(event Custom|ExitVehicleEvent
  ;[1 1 1] Init materials and events on begin play
  (CallFunction|SetMaterial true))

(event Custom|SetMaterial
  (params (SetMaterial :type Boolean))
  ;[1 1 1] SetMaterial On Exit/Enter Vehicle
  (Rendering|Material|SetMaterial (Variables|Visual|GetWeaponMesh) 0 (select SetMaterial (Variables|Default|GetFlare_FX) (Variables|Default|GetNullMaterial))))

(event Custom|OnHolsterSwapMaterials
  (bind _playerpawn (Variables|Default|GetPlayerPawn))
  ;[1 1 1] Nulls out material if you're in the shotgun seat
  (bind _vehicle_occupied (Utilities|IsinVehicle _playerpawn)
    (:true
      (Utilities|FlowControl|Sequence
        (:then_0
          (bind _returnvalue (GameplayTags|HasAllMatchingGameplayTags(Message) (Utilities|ToObject(Interface) _vehicle_occupied) (Variables|Default|GetPickupTruckTag)))
          (if _returnvalue
            (CallFunction|SetMaterial (select (>= (Class|FortVehicleInterface|FindSeatIndex _vehicle_occupied _playerpawn) 2) true false))))
        (:then_1
          (bind _returnvalue_2 (GameplayTags|HasAllMatchingGameplayTags(Message) (Utilities|ToObject(Interface) _vehicle_occupied) (Variables|Default|GetHelicopterTag)))
          (if _returnvalue_2
            (CallFunction|SetMaterial true)))))))

(event EventTick
  (params (DeltaSeconds :type Float%20%28single-precision%29))
  (|CallParentFunction DeltaSeconds) ; This node is disabled and will not be called.\nDrag off pins to build functionality.
)

(fn B_ConsumableSmall_RiftShardJar_Athena_AutoGenFunc
  (params (Pawn :type Pawn%20Object%20Reference)))

(fn ConstructionScript
  (|CallParentFunction))
